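# Physics asset for the Male Large CometHoliday head cosmetic, setting up the hat's left and right flaps for physics simulation.
# Each side has a small capsule on its base flap bone and two larger, overlapping-sized capsules on the dynamic flap bones.
# Constraints chain each side from base through first to second flap segment, all sharing the same tight swing and twist limits for restrained movement.
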PhysicsAsset /BRCosmetics/Characters/Player/Male/Large/Heads/Male_Large_CometHoliday_Head/Meshes/Male_Large_CometHoliday_Head_Physics
bodies {
 base_hatFlap_l [{shapeType=Capsule, length=7.5843, radius=1.9511, shapeName=base_hatFlap_l_capsule, rotation=<rot 0 -90 0>}]
 base_hatFlap_r [{shapeType=Capsule, length=7.5843, radius=1.9511, shapeName=base_hatFlap_r_capsule, rotation=<rot 0 -90 0>}]
 dyn_hatFlap_01_l [{
   shapeType=Capsule
   length=11
   radius=8
   shapeName=dyn_hatFlap_01_l_capsule
   center=<4.7257 -0.2184 -0.4253>
   rotation=<rot 2.6875 -89.9994 -9.6775>
  }]
 dyn_hatFlap_01_r [{
   shapeType=Capsule
   length=11
   radius=8
   shapeName=dyn_hatFlap_01_r_capsule
   center=<5.0797 0.0809 -0.6952>
   rotation=<rot 3.6291 -90 17.8135>
  }]
 dyn_hatFlap_02_l [{
   shapeType=Capsule
   length=11
   radius=8
   shapeName=dyn_hatFlap_02_l_capsule
   center=<5.8001 -0.1868 1.5546>
   rotation=<rot -5.2513 -90 -12.0664>
  }]
 dyn_hatFlap_02_r [{
   shapeType=Capsule
   length=11
   radius=8
   shapeName=dyn_hatFlap_02_r_capsule
   center=<4.6633 0.3129 -1.5003>
   rotation=<rot 1.0511 -89.9999 15.7392>
  }]
}
constraints [
 {bone1Name=dyn_hatFlap_02_r, bone2Name=dyn_hatFlap_01_r, swing1LimitDegrees=10, swing2LimitDegrees=16, twistLimitDegrees=8}
 {bone1Name=dyn_hatFlap_01_l, bone2Name=base_hatFlap_l, swing1LimitDegrees=10, swing2LimitDegrees=16, twistLimitDegrees=8}
 {bone1Name=dyn_hatFlap_02_l, bone2Name=dyn_hatFlap_01_l, swing1LimitDegrees=10, swing2LimitDegrees=16, twistLimitDegrees=8}
 {bone1Name=dyn_hatFlap_01_r, bone2Name=base_hatFlap_r, swing1LimitDegrees=10, swing2LimitDegrees=16, twistLimitDegrees=8}
]
