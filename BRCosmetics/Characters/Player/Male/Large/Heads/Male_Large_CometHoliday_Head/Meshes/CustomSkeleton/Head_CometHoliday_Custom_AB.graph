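# Animation Blueprint for the Comet Holiday large male cosmetic head, layering secondary motion and bone corrections onto the input pose.
# A head constraint to neck_02 and additive hat-flap rotations feed a RigidBody simulation in spine_05 space with overridden gravity.
# A Control Rig follows, then fixed additive offsets on clavicle-push and neck corrective bones.
# Each update computes GravityOverride through a deprecated helper and Hair Up from the spine_05-to-head transform delta.
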AnimBlueprint /BRCosmetics/Characters/Player/Male/Large/Heads/Male_Large_CometHoliday_Head/Meshes/CustomSkeleton/Head_CometHoliday_Custom_AB
target_skeleton=&/Base_Head_CometHoliday_Skeleton
graphs [
 AnimGraph {nodes [
   n1 = "Constraint - head (P)"(ComponentPose=Local To Component(LocalPose=Input Pose()), Alpha=0.8, ConstraintWeights_0=1) {alphaBoolBlend {blendOption=Linear}, boneToModify {boneName=head}, constraintSetup [{targetBone {boneName=neck_02}}], constraintWeights [1]}
   n2 = "Transform (Modify) Bone - Bone: base_hatFlap_l"!(ComponentPose=@n1, bAlphaBoolEnabled:bool=True, Translation="0,0,0", Rotation="0,-10,0", Scale="1,1,1") -> (Pose) {
    alphaInputType=Bool
    rotationMode=BMM_Additive
    rotationSpace=BCS_BoneSpace
    translationSpace=BCS_BoneSpace
    alphaBoolBlend {blendInTime=0.5, blendOption=Linear, blendOutTime=0.5}
    boneToModify {boneName=base_hatFlap_l}
    rotation {yaw=-10}
   }
   n3 = "Transform (Modify) Bone - Bone: base_hatFlap_r"!(ComponentPose=@n2, bAlphaBoolEnabled:bool=True, Translation="0,0,0", Rotation="0,10,0", Scale="1,1,1") -> (Pose) {
    alphaInputType=Bool
    rotationMode=BMM_Additive
    rotationSpace=BCS_BoneSpace
    translationSpace=BCS_BoneSpace
    alphaBoolBlend {blendInTime=0.5, blendOption=Linear, blendOutTime=0.5}
    boneToModify {boneName=base_hatFlap_r}
    rotation {yaw=10}
   }
   n4 = RigidBody!(ComponentPose=@n3, Alpha, OverrideWorldGravity:Vector="0,0,0", ExternalForce) -> (Pose) {
    bClampLinearTranslationLimitToRefPose=true
    bOverrideWorldGravity=true
    overridePhysicsAsset=/BRCosmetics/Characters/Player/Male/Large/Heads/Male_Large_CometHoliday_Head/Meshes/Male_Large_CometHoliday_Head_Physics
    simulationSpace=BaseBoneSpace
    alphaBoolBlend {blendOption=Linear}
    baseBoneRef {boneName=spine_05}
    componentLinearAccScale=<0.1 0.2 0.1>
    componentLinearVelScale=<0.1 0.2 0.1>
   }
   n5 = Control Rig(Source=Component To Local(ComponentPose=@n4)) {
    alphaBoolBlend {blendOption=Linear}
    controlRigAssetReference {blueprintRigClass=/BRCosmetics/Characters/Player/Male/Large/Heads/Male_Large_CometHoliday_Head/Meshes/Male_Large_CometHoliday_Head_CtrlRig.Male_Large_CometHoliday_Head_CtrlRig_C}
   }
   n6 = "Transform (Modify) Bone - Bone: dyn_clavPush_r"!(
    ComponentPose=Local To Component(LocalPose=@n5)
    bAlphaBoolEnabled:bool=True
    Translation="-14.339,0.380,1.297"
    Rotation="-10.401,1.520,13.718"
   ) -> (Pose) {
    alphaInputType=Bool
    rotationMode=BMM_Additive
    rotationSpace=BCS_BoneSpace
    translationMode=BMM_Additive
    translationSpace=BCS_BoneSpace
    alphaBoolBlend {blendInTime=0.1, blendOption=Linear, blendOutTime=0.1}
    boneToModify {boneName=dyn_clavPush_r}
    rotation=<rot -10.401 1.52 13.718>
    translation=<-14.339 0.38 1.297>
   }
   n7 = "Transform (Modify) Bone - Bone: base_neck_corrective_cr"!(ComponentPose=@n6, Alpha=1, Translation="0,0,0", Rotation="30,0,0") -> (Pose) {
    rotationMode=BMM_Additive
    rotationSpace=BCS_BoneSpace
    translationMode=BMM_Additive
    translationSpace=BCS_BoneSpace
    alphaBoolBlend {blendInTime=0.1, blendOption=Linear, blendOutTime=0.1}
    boneToModify {boneName=base_neck_corrective_cr}
    rotation {pitch=30}
   }
   n8 = "Transform (Modify) Bone - Bone: neck_corrective_cr"!(ComponentPose=@n7, Alpha=1, Translation="25.483,-0.166,-0.036", Rotation="0,0,0") -> (Pose) {
    rotationMode=BMM_Additive
    rotationSpace=BCS_BoneSpace
    translationMode=BMM_Additive
    translationSpace=BCS_BoneSpace
    alphaBoolBlend {blendInTime=0.1, blendOption=Linear, blendOutTime=0.1}
    boneToModify {boneName=neck_corrective_cr}
    translation=<25.483 -0.166 -0.036>
   }
   Output Pose(Result=Component To Local(ComponentPose=@n8))
  ]}
 EventGraph {nodes [
   Event BlueprintUpdateAnimation()
   > n1 = Sequence()
   n2 = Get Skel Mesh()
   Set Hair Up(
    execute=@n1.then_1
    Hair Up:double=float * float(
     A=GetAlphaFromReferenceTransformDelta(SkeletalMeshComponent=@n2, BoneName=spine_05, BaseName=head, InputRange=(X=0,Y=90), TwistAxis="0,0,1")
     B=1
    ) {is=title K2Node_PromotableOperator}
   ) -> (Output_Get:double)
   n3 = Gravity Override State Function DEPRECATED(
    execute=@n1.then_0
    CharacterPartAnimInstance=Self-Reference()
    GameTimeInSeconds=GetGameTimeInSeconds()
    GravityOverrideParams=Get Gravity Override Params()
   )
   > Set GravityOverride(GravityOverride=@n3)
   Event BlueprintInitializeAnimation()
   > Set Skel Mesh(Skel Mesh=GetOwnerSkeletalMeshComponent())
  ]}
]
variables [
 Gravity Override Params {type=GravityOverrideParamsStruct}
 GravityOverride {type=Vector}
 Hair Up {type=number}
 Skel Mesh {type=SkeletalMeshComponent}
]
